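# Unlit toon-shaded Niagara mesh particle material for the GrimeHold skydiving contrail, with green-yellow defaults and a radioactive tip tint.
# Stepped light-direction shading, two-tone rim and banded dot highlight combine with particle color.
# A rounded Fresnel cuts a dark outline, and time-of-day functions adjust desaturation and brightness before emissive adjustments.
# Niagara dynamic parameters drive highlights and a distance-masked vertex spike, pulling vertices toward a supplied point without overshooting.
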Material /BRCosmetics/Athena/Effects/Particles/Skydiving/NS_Contrail_GrimeHold/FX/M_Contrail_GrimeHold
graph {
 properties {
  allowTranslucentCustomDepthWrites=true
  bEnableResponsiveAA=true
  bUsedWithNiagaraMeshParticles=true
  pixelDepthOffsetMode=PDOM_Legacy
  shadingModel=MSM_Unlit
  usedShadingModels=Unlit
 }
 nodes [
  n1 = Round(None=Fresnel(ExponentIn=ScalarParameter() {defaultValue=1.5, parameterName=ToonOutlineSpread}) {baseReflectFraction=0, exponent=1})
  n2 = Desaturation(
   None=NamedRerouteUsage()
   Fraction=MaterialFunctionCall(Value_Morning=Constant() {r=0.1}, Value_Day=Constant(), Value_Evening=Constant() {r=0.1}, Value_Night=Constant() {r=0.05}) {materialFunction=&Scalar_Time_of_Day}.Result
  )
  n3 = MaterialFunctionCall(Value_Morning=Constant() {r=0.8}, Value_Day=Constant() {r=1}, Value_Evening=Constant() {r=0.8}, Value_Night=Constant() {r=0.9}) {materialFunction=&Scalar_Time_of_Day}
  MaterialFunctionCall(Emissive=Multiply(A=OneMinus(None=@n1), B=Multiply(A=@n2, B=@n3.Result))) {materialFunction=&MF_FX_EmissiveAdjustments}
  n4 = DynamicParameter() {parameterIndex=1, defaultValue=<rgba 0 0 25 129.41>, paramNames [WPOX,WPOY,WPOZ,WPODistance]}
  n5 = MaterialFunctionCall(X=@n4.WPOX, Y=@n4.WPOX, Z=@n4.WPOX) {materialFunction=&MakeFloat3}
  n6 = Subtract(A=@n5.Result, B=WorldPosition().XYZ)
  n7 = Length(None=@n6) {is=class MaterialExpressionLength}
  n8 = Divide(A=@n7, B=@n4.WPODistance)
  n9 = Power(Base=Saturate(None=OneMinus(None=@n8))) {constExponent=10}
  NamedRerouteDeclaration(None=@n9) {name=WPOMask, nodeColor=<rgba 0.9922 0 1 1>}
  n10 = Multiply(A=@n9, B=DynamicParameter() {paramNames [Specularity, DotHighlightStr, WPOStr, DotBreakAmount]}.WPOStr)
  n11 = Multiply(B=@n7) {constA=0.999}
  n12 = Max(B=@n7) {constA=0.0001}
  NamedRerouteDeclaration(None=Multiply(A=Min(A=@n10, B=@n11), B=Divide(A=@n6, B=@n12))) {name=Vertex Spike, nodeColor=<rgba 1 0.5711 0.1046 1>}
  n13 = VectorParameter() {group=29.Color, parameterName=Color_Rim01, defaultValue=<rgba 1 0.9212 0.0397 1>}
  n14 = VertexNormalWS()
  n15 = CameraVectorWS()
  n16 = VectorParameter() {group=39.RimLight, parameterName=DotHighlight_MaskDir, defaultValue=<rgba 0.5 0 -0.1222 0>}
  n17 = DynamicParameter() {paramNames [Specularity, DotHighlightStr, WPOStr, DotBreakAmount]}
  n18 = Power(Base=DotProduct(A=@n14, B=Add(A=@n15, B=@n16.RGB)), Exp=@n17.DotBreakAmount)
  n19 = ScalarParameter() {defaultValue=2, parameterName=DotHighlight_Spread}
  n20 = Divide(A=Floor(None=Multiply(A=@n18, B=@n19) {constB=2}), B=@n19)
  n21 = Multiply(A=Multiply(A=@n13.RGB, B=Saturate(None=@n20)), B=@n17.DotHighlightStr)
  n22 = VectorParameter() {group=29.Color, parameterName=Color_Rim02, defaultValue=<rgba 0.0965 0.2522 0.0197 1>}
  n23 = DotProduct(A=@n14, B=@n15)
  n24 = Add(A=@n23, B=ScalarParameter() {defaultValue=0.4654, group=39.RimLight, parameterName=RimLight02_Offset})
  n25 = Saturate(None=Step(X=OneMinus(None=@n24)) {constY=0.01})
  n26 = DotProduct(A=NamedRerouteUsage(), B=VertexNormalWS())
  n27 = ScalarParameter() {group=32.LightingGI, parameterName=Lighting_GI_NDS_Offset}
  n28 = Step(X=Saturate(None=Add(A=@n26, B=@n27))) {constY=0.01}
  n29 = Add(A=@n23, B=ScalarParameter() {defaultValue=0.363, group=39.RimLight, parameterName=RimLight_Offset})
  n30 = Saturate(None=Step(X=OneMinus(None=@n29)) {constY=0.01})
  n31 = Multiply(A=Multiply(A=LinearInterpolate(A=@n13.RGB, B=@n22.RGB, Alpha=@n25), B=Multiply(A=@n28, B=@n30)), B=@n17.Specularity)
  n32 = Multiply(
   A=LinearInterpolate(B=VectorParameter() {parameterName=RadioactiveTip_Color, defaultValue=<rgba 0.375 3 0.4188 1>}.RGB, Alpha=NamedRerouteUsage()) {constA=1}
   B=LinearInterpolate(
    A=VectorParameter() {group=29.Color, parameterName=Color_GI, defaultValue=<rgba 0.1463 0.4179 0 1>}.RGB
    B=VectorParameter() {group=29.Color, parameterName=Color_Base, defaultValue=<rgba 0.1946 0.4179 0.0176 1>}.RGB
    Alpha=Saturate(None=@n28)
   )
  )
  n33 = ParticleColor()
  NamedRerouteDeclaration(None=Add(A=LinearInterpolate(A=@n21, Alpha=@n31), B=Multiply(A=@n32, B=@n33.RGB))) {name=LightingOutEmissive, nodeColor=<rgba 0.4969 0.2384 1 1>}
  n34 = VectorParameter() {parameterName=FELightDir}
  NamedRerouteDeclaration(None=LinearInterpolate(A=SkyAtmosphereLightDirection(), B=@n34.RGB, Alpha=@n34.RGB)) {name=LightDir, nodeColor=<rgba 0 0.2031 1 1>}
  NamedRerouteUsage()
  n35 = VertexNormalWS()
  Saturate(None=OneMinus(None=DotProduct(A=@n5.Result, B=@n35)))
 ]
}
