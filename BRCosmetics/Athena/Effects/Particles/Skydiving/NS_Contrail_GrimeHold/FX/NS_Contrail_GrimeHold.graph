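# GrimeHold skydiving contrail Niagara system that adds socket-attached meshes, green glow sprites and smoke to the player character.
# Mesh pins two particles to skeletal mesh sockets, while Smoke_Large spawns shrinking mesh puffs from those sockets with sine-wobbled, rotated offsets.
# Paired CPU and GPU emitters add velocity-stretched speedlines, with spawn rate scaled by an integrator.
# Dive attributes blend offsets and velocities, and front-end switches retune values for menus.
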NiagaraSystem /BRCosmetics/Athena/Effects/Particles/Skydiving/NS_Contrail_GrimeHold/FX/NS_Contrail_GrimeHold
dependencies {
 usedDataInterfaces [
  &NiagaraDataInterfaceCurve
  &NiagaraDataInterfaceVector2DCurve
  &NiagaraDataInterfaceArrayFloat3
  &NiagaraDataInterfacePlatformSet
  &NiagaraDataInterfaceSkeletalMesh
 ]
 usedDynamicInputs [
  &SimulationPosition
  &UniformRangedVector2D
  &TransformPosition
 ]
 usedModules [
  &self:SystemSpawnScript.SystemScriptSource.SystemScriptGraph.NiagaraNodeAssignment_6.SetVariables_#1
  &SystemState
  /Game/Effects/Niagara/Modules/Skydiving/FN_DivingAttributes/v5/FN_DivingAttributes
  &FN_Integrator
  &self:SystemSpawnScript.SystemScriptSource.SystemScriptGraph.NiagaraNodeAssignment_3.SetVariables_#2
 ]
}
emitters {
 Mesh {
  emitterUpdateScript [
   &EmitterState(
    Inactive Response=::NewEnumerator1
    Life Cycle Mode=::NewEnumerator1
    Loop Behavior=::NewEnumerator1
    Loop Duration Mode=::NewEnumerator1
    UseLoopDelay=false
   )
   &SpawnBurst_Instantaneous(Spawn Count=2, Spawn Group=1, Use Spawn Probability=false, UseLoopCountLimit=false) {module=SpawnBurst_Instantaneous002}
   &SpawnBurst_Instantaneous(Spawn Count=2, Use Spawn Probability=false, UseLoopCountLimit=false)
  ]
  particleSpawnScript [
   &FN_SetSpawnGroup(Number of Spawn Groups=2)
   &InitializeParticle(
    Mesh Scale=<1.3 1.3 1.3>
    Mesh Scale Mode=::NewEnumerator4
    Sprite Size=&MergeHelper_Vector2Switch(0=<38.6824 79.2701>, 1=<38.6824 79.2701>, 2=?, 3=?, 4=?, Index=$Emitter.SpawnGroup, Num=::NewEnumerator0)
    Sprite Size Mode=::NewEnumerator4
    UsePositionOffset=false
    Write Material Random=true
   )
   &InitialMeshOrientation(
    Orientation Axis=<0 0 0>
    Orientation Vector=<0 0 0>
    Rotation=&Lerp_Vector(A=<0 0 0>, B=<0 0.5 0>, Alpha=&MakeFloatFromInt(Int=&SpawnGroup()))
    Rotation Coordinate Space=::NewEnumerator0
    Use Orientation Vector=true
    Use Rotation Vector=true
   )
   /SetVariables(Particles.VisibilityTag=$Emitter.SpawnGroup) {isSetParameters=true}
   /SetVariables(Particles.DecalGlowCameraOffset=5, Particles.GlowCameraOffset=5) {isSetParameters=true}
   /SetVariables(Particles.NScale:Vector3f=<1.5 1.5 1.3>) {isSetParameters=true}
  ]
  particleUpdateScript [
   &SolveForcesAndVelocity(Clamp Velocity=false, Limit Acceleration=false)
   &ParticleState(Kill Particles When Lifetime Has Elapsed=false)
   &SkeletalMeshLocation(
    Apply Sampled Orientation to Mesh Particles=true
    Bone / Socket Index=&SpawnGroup()
    Inherit Velocity=false
    Interpolate Mesh Sampling=false
    MaskSpawnGroup=false
    Mesh Sampling Type=::NewEnumerator5
    Sampled Position Offset=&MergeHelper_VectorSwitch(
     0=&SelectVectorFromArray(
      Direct Array Index=&SpawnGroup()
      Vector Selection Array=DataInterface() {propertyValues {FloatData [<0 -8 2>, <0 8 -1>]}}
     )
     1=&SelectVectorFromArray(
      Direct Array Index=&SpawnGroup()
      Vector Selection Array=DataInterface() {propertyValues {FloatData [<0 -7.3211 2.3325>, <0 8.719 -0.9948>]}}
     )
     2=&SelectVectorFromArray(
      Direct Array Index=&SpawnGroup()
      Vector Selection Array=DataInterface() {propertyValues {FloatData [<-20 -2.019 23.9132>, <20 1.4248 -23.9132>]}}
     )
     3=?
     Index=$Particles.VisibilityTag
     Num=::NewEnumerator0
    )
    Sampled Position Offset Coordinate Space:ENiagaraMeshTransforms=::NewEnumerator3
    Skeletal Mesh=$System.Skeletal Mesh
    Socket Sampling Mode=::NewEnumerator5
   )
   &ScaleColor(
    Scale Alpha=&Multiply_Float(
     A=&Multiply_Float(
      A=&FloatFromCurve(
       FloatCurve=DataInterface() {propertyValues {Curve {keys [<at 0.2 0>,<at 0.7 1>]}}}
       CurveIndex=$System.Age
       Scale Curve=&FN_FrontEndFloatModifier(
        Gameplay=0.1
        Front End=&FN_PickFloatDependOnPlatform(TargetPlatform=DataInterface() {propertyValues {Platforms {qualityLevelMask=-29}}}, TargetPlatformActive=0.05, TargetPlatformNoActive=0.03)
       )
      )
      B=0.85
     )
     B=1
    )
    Scale RGB=&MakeVectorFromLinearColorRGB(Color=<rgba 0 2.2 0.0367 1>)
    ScaleA=true
    ScaleRGB=true
   )
   &DynamicMaterialParameters(Write Parameter Index 0=::NewEnumerator0)
   &SpriteFacingAndAlignment(
    Alignment Coordinate Space=::Local
    Sprite Alignment=&RotateVectorByQuaternion(
     Vector To Rotate=&MergeHelper_VectorSwitch(0=<0 0 1>, 1=<0 0 1>, 2=?, 3=?, Index=$Particles.VisibilityTag, Num=::NewEnumerator0)
     Quaternion=$Output.SkeletalMeshLocation.SampledRotation
    )
    WriteAlignment=true
    WriteFacing=true
   ) {module=SpriteFacingAndAlignment001}
   &FN_Sprite_Attributes(
    Color=&Multiply_LinearColorByFloat(Linear Color=$Particles.Color, Float=&Lerp_Float(A=0.15, B=0.2, Alpha=$System.Dive))
    Create Color Attribute=true
    Create Position Attribute=true
    Create SpriteSize Attribute=true
    Position:NiagaraPosition=&AddVectorToPosition(
     Position=$Particles.Position
     Vector=&MergeHelper_VectorSwitch(
      0=&Lerp_Vector(A=<16.1986 0 0>, B=<0 0 15>, Alpha=$System.RDive)
      1=&Lerp_Vector(A=<-16.1986 0 0>, B=<0 0 15>, Alpha=$System.RDive)
      2=?
      3=?
      Index=$Particles.SpawnGroup
      Num=::NewEnumerator0
     )
    )
    SpriteSize=&Multiply_Vector2D_ByFloat(Float=50, Vector2D=<1 1>)
    Write to Module Namespace=true
   )
  ]
  rendererClasses [&NiagaraMeshRendererProperties, &NiagaraSpriteRendererProperties]
  renderers [
   {rendererClass=&NiagaraMeshRendererProperties, rendererIndex=0}
   {rendererClass=&NiagaraSpriteRendererProperties, rendererIndex=1}
   {rendererClass=&NiagaraSpriteRendererProperties, rendererIndex=2}
   {rendererClass=&NiagaraMeshRendererProperties, rendererIndex=3}
  ]
 }
 Smoke_Large {
  emitterUpdateScript [
   &EmitterState(Life Cycle Mode=::NewEnumerator1, Loop Duration=2, UseLoopDelay=false)
   &SpawnBurst_Instantaneous(Spawn Count=2, Spawn Group=1, Use Spawn Probability=false, UseLoopCountLimit=false)
   &SpawnRate(SpawnRate=&Lerp_Float(A=30, B=40, Alpha=$System.RDive), Use Spawn Probability=false)
  ]
  particleSpawnScript [
   &FN_SetSpawnGroup(Number of Spawn Groups=2)
   &InitializeParticle(
    Lifetime=&MergeHelper_FloatSwitch(
     0=&Lerp_Float(
      A=&RandomRangeFloat(Minimum=0.7, Maximum=0.8)
      B=&RandomRangeFloat(Minimum=0.65, Maximum=0.78)
      Alpha=$System.RDive
     )
     1=3
     2=?
     3=?
     Index=$Emitter.SpawnGroup
     Num=::NewEnumerator0
    )
    Mesh Scale Mode=::NewEnumerator3
    Mesh Uniform Scale=&Multiply_Float(A=&RandomRangeFloat(Minimum=0.5, Maximum=1.1), B=1.35)
    Position Offset=&Lerp_Vector(
     A=<0 0 -30>
     B=&FN_FrontEndVector3Modifier(Gameplay=<0 30 0>, Front End=<0 30 -15>)
     Alpha=$System.Dive
    )
    UsePositionOffset=true
    Write Material Random=true
   )
   &AddVelocity(
    Cone Angle=&MergeHelper_FloatSwitch(0=3, 1=1, 2=?, 3=?, Index=$Emitter.SpawnGroup, Num=::NewEnumerator0)
    Cone Axis=&MaskVectorBySpawnGroup(Masked Vector=<0.1 0 1>, Passthrough Vector=<-0.1 0 1>)
    Module.UseConeFalloff=false
    Module.UseVelDistribution=false
    Override Seed=false
    Velocity Mode=::NewEnumerator2
    Velocity Speed=&MergeHelper_FloatSwitch(
     0=&Lerp_Float(A=130, B=170, Alpha=$System.Dive)
     1=&Lerp_Float(A=115, B=150, Alpha=$System.Dive)
     2=?
     3=?
     Index=$Emitter.SpawnGroup
     Num=::NewEnumerator0
    )
   )
   /SetVariables(Particles.MeshIndex=&RandomRangeInt(Minimum=0, Maximum=1)) {isSetParameters=true}
   &InitialMeshOrientation(Use Rotation Vector=false)
   /SetVariables(Particles.VisibilityTag=$Emitter.SpawnGroup) {isSetParameters=true}
   &AddVelocity(Velocity=&Lerp_Vector(
     A=&Lerp_Vector(A=<200 0 0>, B=<0 150 0>, Alpha=$System.Dive)
     B=&Lerp_Vector(A=<-200 0 0>, B=<0 150 0>, Alpha=$System.Dive)
     Alpha=&MakeFloatFromInt(Int=$Particles.SpawnGroup)
    )) {module=AddVelocity001}
  ]
  particleUpdateScript [
   &ParticleState()
   &UpdateMeshOrientation(Rotation Vector=&RandomRangeVector(Minimum=<0 0 0>, Maximum=<1 1 1>))
   &ScaleVelocity(Velocity Scale=&MakeVector(
     X=&FloatFromCurve(FloatCurve=DataInterface() {propertyValues {Curve {keys [
          {arriveTangent=-2.7958, interpMode=RCIM_Cubic, time=0.2, value=1}
          {arriveTangent=-1.8824, interpMode=RCIM_Cubic, leaveTangent=-1.8824, tangentMode=RCTM_User, time=1, value=0}
         ]}}})
     Y=1
     Z=&FloatFromCurve(FloatCurve=DataInterface() {propertyValues {Curve {keys [{arriveTangent=2.6979, interpMode=RCIM_Cubic, leaveTangent=2.6979, tangentMode=RCTM_User, time=0, value=0.1}, <at 1 2>]}}})
    ))
   &SolveForcesAndVelocity(Clamp Velocity=false, Limit Acceleration=false)
   &SkeletalMeshLocation(
    Bone / Socket Index=$Particles.SpawnGroup
    MaskSpawnGroup=false
    Mesh Sampling Type=::NewEnumerator5
    Orientation Sampling=::NewEnumerator1
    Position Sampling=::NewEnumerator2
    Sampled Position Offset=&Add_Vector(
     A=&MergeHelper_VectorSwitch(
      0=&SelectVectorFromArray(
       Direct Array Index=&SpawnGroup()
       Vector Selection Array=DataInterface() {propertyValues {FloatData [<0 -8 2>, <0 8 -1>]}}
      )
      1=&SelectVectorFromArray(
       Direct Array Index=&SpawnGroup()
       Vector Selection Array=DataInterface() {propertyValues {FloatData [<0 -7.3211 2.3325>, <0 8.719 -0.9948>]}}
      )
      2=&SelectVectorFromArray(
       Direct Array Index=&SpawnGroup()
       Vector Selection Array=DataInterface() {propertyValues {FloatData [<-20 -2.019 23.9132>, <20 1.4248 -23.9132>]}}
      )
      3=?
      Index=$Particles.SpawnGroup
      Num=::NewEnumerator0
     )
     B=&MergeHelper_VectorSwitch(
      0=&Lerp_Vector(A=<0 0 18.8699>, B=<0 0 18.8699>, Alpha=$System.RDive)
      1=&Lerp_Vector(A=<0 0 -18.8699>, B=<0 0 -18.8699>, Alpha=$System.RDive)
      2=?
      3=?
      Index=$Particles.SpawnGroup
      Num=::NewEnumerator0
     )
    )
    Skeletal Mesh=$System.Skeletal Mesh
    Socket Sampling Mode=::NewEnumerator5
    Use Sampled Mesh Attribute as Ribbon ID=false
   )
   /SetVariables(Particles.NPosition:Vector3f=&Add_Vector(
     A=&Add_Vector(
      A=Linked() {linkedVariable {name=Particles.Position, type=NiagaraPosition}}
      B=&FN_FrontEndVector3Modifier(
       Gameplay=&Lerp_Vector(
        A=<0 0 30>
        B=&MaskVectorBySpawnGroup(Masked Vector=<5 -6 30>, Passthrough Vector=<-8 -6 30>)
        Alpha=$System.RDive
       )
       Front End=&Lerp_Vector(
        A=&MaskVectorBySpawnGroup(Masked Vector=<0 0 30>, Passthrough Vector=<-5 0 30>)
        B=&MaskVectorBySpawnGroup(Masked Vector=<5 -6 30>, Passthrough Vector=<-8 -6 30>)
        Alpha=$System.RDive
       )
      )
     )
     B=&Multiply_VectorByFloat(
      Vector=&Lerp_Vector(
       A=<0 0 0>
       B=<7 0 0>
       Alpha=&ClampFloat(Float=&Divide_Float(
         A=&MakeFloatFromVector(Vector=Linked() {linkedVariable {name=Particles.Position, type=NiagaraPosition}}, Channel=::NewEnumerator2)
         B=60
        ))
      )
      Float=&Sine(
       Normalized Angle=&Add_Float(
        A=&Add_Float(
         A=&Multiply_Float(A=$System.Age, B=1)
         B=&MakeCustomFloatFromBool(
          Bool=&SetBoolByIntComparison(A=&SpawnGroup(), B=0, Comparison Type=::NewEnumerator4)
          True Float=0.43
          False Float=0
         )
        )
        B=&Divide_Float(
         A=&MakeFloatFromVector(Vector=Linked() {linkedVariable {name=Particles.Position, type=NiagaraPosition}}, Channel=::NewEnumerator2)
         B=&MaskFloatBySpawnGroup(Masked Float=100, Passthrough Float=120)
        )
       )
       Period=1
       Scale=1
       Bias=0
      )
     )
    )) {isSetParameters=true}
   /SetVariables(Particles.NPosition:Vector3f=&Add_Vector(
     A=/Niagara/DynamicInputs/Math/TransformPositionByMatrix(
      Position:Vector3f=$Particles.NPosition
      TransformationMatrix:NiagaraMatrix=/Niagara/DynamicInputs/Math/MakeTransform(
       Inverse:NiagaraBool=false
       Position:Vector3f=?
       Local Offset:Vector3f=<0 0 0>
       Scale:Vector3f=<1 1 1>
       Rotation Interface:ENiagara_RotationDefinitionApproach=::NewEnumerator2
       Axis:Vector3f=?
       Quaternion:Quat4f=?
       Rotation Matrix:NiagaraMatrix=?
       X Vector:Vector3f=?
       Yaw:NiagaraFloat=0
       Angle:NiagaraFloat=?
       Pitch:NiagaraFloat=&Lerp_Float(
        A=&MakeCustomFloatFromBool(
         Bool=&SetBoolByIntComparison(A=$Particles.SpawnGroup, B=0, Comparison Type=::NewEnumerator4)
         True Float=-5
         False Float=5
        )
        B=&MakeCustomFloatFromBool(
         Bool=&SetBoolByIntComparison(A=$Particles.SpawnGroup, B=0, Comparison Type=::NewEnumerator4)
         True Float=-5.5
         False Float=5.5
        )
        Alpha=$System.Dive
       )
       Y Vector:Vector3f=?
       Roll:NiagaraFloat=&Lerp_Float(
        A=-10
        B=&MakeCustomFloatFromBool(Bool=$User.bIsFrontEndPreview, True Float=-30, False Float=0)
        Alpha=$System.Dive
       )
       Z Vector:Vector3f=?
      )
     )
     B=Linked() {linkedVariable {name=Particles.SkeletalMeshLocation.SampledPosition, type=NiagaraPosition}}
    )) {isSetParameters=true}
   &ScaleMeshSize(Scale Factor=&VectorFromFloat(Value=&FloatFromCurve(FloatCurve=DataInterface() {propertyValues {Curve {keys [
          <at 0 1 RCIM_Cubic>
          {arriveTangentWeight=0.0867, interpMode=RCIM_Cubic, leaveTangentWeight=0.156, tangentMode=RCTM_User, time=0.25, value=1}
          {arriveTangent=-1.0517, interpMode=RCIM_Cubic, leaveTangent=-1.0517, tangentMode=RCTM_User, time=0.7, value=0.8}
          {arriveTangent=-0.8076, interpMode=RCIM_Cubic, leaveTangent=-0.8076, tangentMode=RCTM_User, time=1, value=0}
         ]}}})))
   &ScaleMeshSize(Scale Factor=&VectorFromFloat(Value=&RemapRange(
      Input Value To Remap=&Sine(
       Normalized Angle=&Add_Float(
        A=$Particles.Age
        B=&MakeCustomFloatFromBool(
         Bool=&SetBoolByIntComparison(A=&SpawnGroup(), B=0, Comparison Type=::NewEnumerator4)
         True Float=0.9303
         False Float=0
        )
       )
       Period=0.6
       Scale=1
       Bias=0
      )
      Input Min=-1
      Input Max=1
      Output Min=0.8
      Output Max=1
     ))) {module=ScaleMeshSize002}
   &ScaleColor(Scale RGB=&VectorFromFloat(Value=2.2), ScaleA=true, ScaleRGB=true)
   &DynamicMaterialParameters(
    Index 0 Param 1=&FloatFromCurve(FloatCurve=DataInterface() {propertyValues {Curve {keys [
         {arriveTangent=7.9963, interpMode=RCIM_Cubic, leaveTangent=7.9963, tangentMode=RCTM_User, time=0.125, value=0}
         {
          arriveTangent=-2.0778
          arriveTangentWeight=0.1496
          interpMode=RCIM_Cubic
          leaveTangent=-2.0778
          leaveTangentWeight=0.2721
          tangentMode=RCTM_User
          time=0.4
          value=1
         }
         <at 1 0 RCIM_Cubic>
        ]}}})
    Index 0 Param 2=&FloatFromCurve(
     FloatCurve=DataInterface() {propertyValues {Curve {keys [
         {arriveTangent=7.9963, interpMode=RCIM_Cubic, time=0.125, value=0}
         {
          arriveTangent=4.7622
          arriveTangentWeight=0.1496
          interpMode=RCIM_Cubic
          leaveTangent=4.7622
          leaveTangentWeight=0.2721
          tangentMode=RCTM_User
          time=0.4
          value=0.95
         }
         <at 1 0 RCIM_Cubic>
        ]}}}
     Scale Curve=&MakeCustomFloatFromBool(
      Bool=&SetBoolByFloatComparison(A=&RandomRangeFloat(Minimum=0, Maximum=1), B=0.3)
      True Float=1
      False Float=0
     )
    )
    Index 0 Param 3=200
    Index 0 Param 4=&RandomRangeFloat(Minimum=0.2, Maximum=0.35)
    Index 1 Param 4=130
    Index 1 Vector 1:Vector3f=&TransformVector(Vector=$Particles.SkeletalMeshLocation.SampledPosition, Destination Space=::World, Transform Vector as Position=true)
    Module.Index0Float0Write=true
    Module.Index0Float1Write=true
    Module.Index0Float2Write=true
    Module.Index0Float3Write=true
    Module.Index1Float3Write=true
    Module.Index1VectorWrite=true
    Write Parameter Index 1=::NewEnumerator3
   )
   /SetVariables(Particles.DynamicParameterAlt:Vector4f=<1 1 1 1>) {isSetParameters=true}
   &FN_Sprite_Attributes(
    Camera Offset=5
    Color=&MakeLinearColorFromVectorAndFloat(
     "Vector (RGB)"=&MakeVectorFromLinearColorRGB(Color=<rgba 0 2.2 0.0367 1>)
     "Float (Alpha)"=&Multiply_Float(
      A=0.85
      B=&FloatFromCurve(
       FloatCurve=DataInterface() {propertyValues {Curve {keys [<at 0 0 RCIM_Cubic>,<at 0.2 0.7 RCIM_Cubic>,<at 1 0 RCIM_Cubic>]}}}
       Scale Curve=&FN_FrontEndFloatModifier(
        Gameplay=0.1
        Front End=&FN_PickFloatDependOnPlatform(TargetPlatform=DataInterface() {propertyValues {Platforms {qualityLevelMask=-29}}}, TargetPlatformActive=0.05, TargetPlatformNoActive=0.03)
       )
      )
     )
    )
    Create Camera Offset Attribute=true
    Create Color Attribute=true
    Create Position Attribute=true
    Create SpriteSize Attribute=true
    Position:NiagaraPosition=&AddVectorToPosition(Position=Linked() {linkedVariable {name=Particles.SkeletalMeshLocation.SampledPosition, type=Vector3f}}, Vector=<0 0 5>)
    SpriteSize=<40 50>
    Write to Module Namespace=true
   )
  ]
  rendererClasses [&NiagaraMeshRendererProperties, &NiagaraSpriteRendererProperties]
  renderers [
   {rendererClass=&NiagaraMeshRendererProperties, rendererIndex=0}
   {rendererClass=&NiagaraSpriteRendererProperties, rendererIndex=1}
  ]
 }
 Speedlines_CPU {
  emitterUpdateScript [
   &EmitterState(Life Cycle Mode=::NewEnumerator1, Loop Duration=1, UseLoopDelay=false)
   &SpawnRate(
    SpawnRate=&Multiply_Float(
     A=$System.FN_Integrator001.Value
     B=&FN_FrontEndFloatModifier(Gameplay=&UniformRangedFloat(Minimum=10, Maximum=20), Front End=0)
    )
    Use Spawn Probability=false
   )
  ]
  particleSpawnScript [
   &InitializeParticle(
    Color=&MakeLinearColorFromVectorAndFloat(
     "Vector (RGB)"=<1 1 1>
     "Float (Alpha)"=&Multiply_Float(
      A=&Multiply_Float(
       A=&UniformRangedFloat(Minimum=0.015, Maximum=0.15)
       B=&FN_FrontEndFloatModifier(Gameplay=1, Front End=0.5)
      )
      B=&Add_Float(
       A=1
       B=&Multiply_Float(A=10, B=&Subtract_Float(A=1, B=$Engine.Emitter.SpawnCountScale))
      )
     )
    )
    Lifetime=&FN_FrontEndFloatModifier(
     Gameplay=&UniformRangedFloat(Minimum=0.15, Maximum=0.3)
     Front End=&UniformRangedFloat(Minimum=0.1, Maximum=0.2)
    )
    Sprite Size Max=<5 6>
    Sprite Size Min=<3 4>
    Sprite Size Mode=::NewEnumerator2
    UsePositionOffset=false
    Write Material Random=true
   ) {module=InitializeParticle002}
   &CylinderLocation(
    Cylinder Radius=&UniformRangedFloat(Minimum=100, Maximum=150)
    MaskSpawnGroup=false
    Override Seed=false
    UseOwnerScale=true
   )
   &AddVelocity(Velocity=&FN_FrontEndVector3Modifier(Gameplay=<0 0 1000>, Front End=<0 0 3000>))
   &AddVelocityFromPoint(
    Module.Use Falloff Distance=false
    Velocity Strength=&Multiply_Float(
     A=&FN_FrontEndFloatModifier(Gameplay=1, Front End=2)
     B=&UniformRangedFloat(Minimum=500, Maximum=700)
    )
   )
  ]
  particleUpdateScript [
   &ParticleState()
   &SolveForcesAndVelocity(Clamp Velocity=false, Limit Acceleration=false)
   &ScaleSpriteSize(Scale Factor=&Vector2DFromCurve(
     Vector2Curve=DataInterface() {propertyValues {XCurve {keys [<at 0 0.5 RCIM_Cubic>,<at 0.5 1 RCIM_Cubic>]}, YCurve {keys [<at 0 0 RCIM_Cubic>,<at 0.5 1 RCIM_Cubic>]}}}
     ScaleCurve=&Vector2DFromFloat(Value=1)
    ))
   &ScaleSpriteSizeBySpeed(Max Scale Factor=<1 40>, Min Scale Factor=<0.01 0.05>, Velocity Threshold=1)
   &ScaleColor(
    Scale Alpha=&FloatFromCurve(
     FloatCurve=DataInterface() {propertyValues {Curve {keys [<at 0 0 RCIM_Cubic>, <at 0.2 1>, {arriveTangent=-2.221, interpMode=RCIM_Cubic, leaveTangent=-2.221, tangentMode=RCTM_User, time=1, value=0}]}}}
     Scale Curve=&FN_FrontEndFloatModifier(Gameplay=0.4, Front End=0)
    )
    ScaleA=true
    ScaleRGB=true
   )
  ]
  rendererClasses [&NiagaraSpriteRendererProperties]
  renderers [{rendererClass=&NiagaraSpriteRendererProperties}]
 }
 Speedlines_GPU {
  simTarget=GPUComputeSim
  emitterUpdateScript [
   &EmitterState(Life Cycle Mode=::NewEnumerator1, Loop Duration=1, UseLoopDelay=false)
   &SpawnRate(
    SpawnRate=&Multiply_Float(
     A=$System.FN_Integrator001.Value
     B=&FN_FrontEndFloatModifier(Gameplay=&UniformRangedFloat(Minimum=100, Maximum=200), Front End=0)
    )
    Use Spawn Probability=false
   )
  ]
  particleSpawnScript [
   &InitializeParticle(
    Color=&MakeLinearColorFromVectorAndFloat(
     "Vector (RGB)"=<1 1 1>
     "Float (Alpha)"=&Multiply_Float(
      A=&Multiply_Float(
       A=&UniformRangedFloat(Minimum=0.015, Maximum=0.15)
       B=&FN_FrontEndFloatModifier(Gameplay=1, Front End=0.5)
      )
      B=&Add_Float(
       A=1
       B=&Multiply_Float(A=10, B=&Subtract_Float(A=1, B=$Engine.Emitter.SpawnCountScale))
      )
     )
    )
    Lifetime=&FN_FrontEndFloatModifier(
     Gameplay=&UniformRangedFloat(Minimum=0.15, Maximum=0.3)
     Front End=&UniformRangedFloat(Minimum=0.1, Maximum=0.2)
    )
    Sprite Size Max=<5 6>
    Sprite Size Min=<3 4>
    Sprite Size Mode=::NewEnumerator2
    UsePositionOffset=false
    Write Material Random=true
   ) {module=InitializeParticle002}
   &CylinderLocation(
    Cylinder Radius=&UniformRangedFloat(Minimum=100, Maximum=150)
    MaskSpawnGroup=false
    Override Seed=false
    UseOwnerScale=true
   )
   &AddVelocity(Velocity=&FN_FrontEndVector3Modifier(Gameplay=<0 0 1000>, Front End=<0 0 3000>))
   &AddVelocityFromPoint(
    Module.Use Falloff Distance=false
    Velocity Strength=&Multiply_Float(
     A=&FN_FrontEndFloatModifier(Gameplay=1, Front End=2)
     B=&UniformRangedFloat(Minimum=500, Maximum=700)
    )
   )
  ]
  particleUpdateScript [
   &ParticleState()
   &SolveForcesAndVelocity(Clamp Velocity=false, Limit Acceleration=false)
   &ScaleSpriteSize(Scale Factor=&Vector2DFromCurve(
     Vector2Curve=DataInterface() {propertyValues {XCurve {keys [<at 0 0.5 RCIM_Cubic>,<at 0.5 1 RCIM_Cubic>]}, YCurve {keys [<at 0 0 RCIM_Cubic>,<at 0.5 1 RCIM_Cubic>]}}}
     ScaleCurve=&Vector2DFromFloat(Value=1)
    ))
   &ScaleSpriteSizeBySpeed(Max Scale Factor=<1 40>, Min Scale Factor=<0.01 0.05>, Velocity Threshold=1)
   &ScaleColor(
    Scale Alpha=&FloatFromCurve(
     FloatCurve=DataInterface() {propertyValues {Curve {keys [<at 0 0 RCIM_Cubic>, <at 0.2 1>, {arriveTangent=-2.221, interpMode=RCIM_Cubic, leaveTangent=-2.221, tangentMode=RCTM_User, time=1, value=0}]}}}
     Scale Curve=&FN_FrontEndFloatModifier(Gameplay=0.4, Front End=0)
    )
    ScaleA=true
    ScaleRGB=true
   )
  ]
  rendererClasses [&NiagaraSpriteRendererProperties]
  renderers [{rendererClass=&NiagaraSpriteRendererProperties}]
 }
}
summary {
 systemName=NS_Contrail_GrimeHold
 emitters [
  {emitterName=Speedlines_GPU, simTarget=GPUComputeSim, rendererClasses [&NiagaraSpriteRendererProperties]}
  {emitterName=Smoke_Large, rendererClasses [&NiagaraMeshRendererProperties, &NiagaraSpriteRendererProperties]}
  {emitterName=Mesh, rendererClasses [&NiagaraMeshRendererProperties, &NiagaraSpriteRendererProperties]}
  {emitterName=Speedlines_CPU, rendererClasses [&NiagaraSpriteRendererProperties]}
 ]
 userVariables [
  User.bIsFrontEnd {type=NiagaraBool}
  User.bIsFrontEndPreview {type=NiagaraBool, defaultValue=true}
  User.ParaGlideLeanAlpha {type=NiagaraFloat, defaultValue=0, description=""}
 ]
}
